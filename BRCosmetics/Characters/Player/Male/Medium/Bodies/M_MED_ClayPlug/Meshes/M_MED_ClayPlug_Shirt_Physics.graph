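# Physics asset for the shirt mesh of the male medium ClayPlug cosmetic body, simulating the hanging shirt strips around the hips.
# Nine strips (one center, four per side) each pair a tiny base capsule, angularly free against the pelvis, with a larger dyn capsule on tight swing limits.
# Most strips lock or zero twist, while the center allows moderate twist.
# Pelvis and thigh capsules supply the lower-body volumes.
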PhysicsAsset /BRCosmetics/Characters/Player/Male/Medium/Bodies/M_MED_ClayPlug/Meshes/M_MED_ClayPlug_Shirt_Physics
bodies {
 base_shirt_shered_01_c [{shapeType=Capsule, length=0.505, radius=0.505, shapeName=base_shirt_shered_01_c_capsule, rotation=<rot 0 -90 0>}]
 base_shirt_shered_01_l [{shapeType=Capsule, length=0.505, radius=0.505, shapeName=base_shirt_shered_01_l_capsule, rotation=<rot 0 -90 0>}]
 base_shirt_shered_01_r [{shapeType=Capsule, length=0.505, radius=0.505, shapeName=base_shirt_shered_01_r_capsule, rotation=<rot 0 -90 0>}]
 base_shirt_shered_02_l [{shapeType=Capsule, length=0.505, radius=0.505, shapeName=base_shirt_shered_02_l_capsule, rotation=<rot 0 -90 0>}]
 base_shirt_shered_02_r [{shapeType=Capsule, length=0.505, radius=0.505, shapeName=base_shirt_shered_02_r_capsule, rotation=<rot 0 -90 0>}]
 base_shirt_shered_03_l [{shapeType=Capsule, length=0.505, radius=0.505, shapeName=base_shirt_shered_03_l_capsule, rotation=<rot 0 -90 0>}]
 base_shirt_shered_03_r [{shapeType=Capsule, length=0.505, radius=0.505, shapeName=base_shirt_shered_03_r_capsule, rotation=<rot 0 -90 0>}]
 base_shirt_shered_04_l [{shapeType=Capsule, length=0.505, radius=0.505, shapeName=base_shirt_shered_04_l_capsule, rotation=<rot 0 -90 0>}]
 base_shirt_shered_04_r [{shapeType=Capsule, length=0.505, radius=0.505, shapeName=base_shirt_shered_04_r_capsule, rotation=<rot 0 -90 0>}]
 dyn_shirt_shered_01_c [{
   shapeType=Capsule
   length=6.3042
   radius=2.3563
   shapeName=dyn_shirt_shered_01_c_capsule
   center=<6.4201 -1.4096 0>
   rotation=<rot 0 -0.0001 0>
  }]
 dyn_shirt_shered_01_l [{shapeType=Capsule, length=6, radius=2, shapeName=dyn_shirt_shered_01_l_capsule, center=<6.9024 -0.5341 0>, rotation=<rot 0 -10.0003 0>}]
 dyn_shirt_shered_01_r [{shapeType=Capsule, length=6, radius=2, shapeName=dyn_shirt_shered_01_r_capsule, center=<7.768 -0.6867 0>, rotation=<rot 0 169.9998 180>}]
 dyn_shirt_shered_02_l [{shapeType=Capsule, length=6, radius=2, shapeName=dyn_shirt_shered_02_l_capsule, center=<3.8735 0 0>, rotation=<rot 90 98.5953 108.5954>}]
 dyn_shirt_shered_02_r [{shapeType=Capsule, length=6, radius=2, shapeName=dyn_shirt_shered_02_r_capsule, center=<3.8735 0 0>, rotation=<rot 90 95.6613 105.6612>}]
 dyn_shirt_shered_03_l [{shapeType=Capsule, length=6, radius=2, shapeName=dyn_shirt_shered_03_l_capsule, center=<3.8735 0 0>, rotation=<rot 90 88.5471 98.5471>}]
 dyn_shirt_shered_03_r [{shapeType=Capsule, length=6, radius=2, shapeName=dyn_shirt_shered_03_r_capsule, center=<3.8735 0 0>, rotation=<rot 90 93.2459 103.2457>}]
 dyn_shirt_shered_04_l [{shapeType=Capsule, length=6, radius=2, shapeName=dyn_shirt_shered_04_l_capsule, center=<3.8735 0 0>, rotation=<rot 90 92.9512 102.951>}]
 dyn_shirt_shered_04_r [{
   shapeType=Capsule
   length=8.9683
   radius=2.3655
   shapeName=dyn_shirt_shered_04_r_capsule
   center=<6.011 0 0>
   rotation=<rot 90 -172.4476 -172.4477>
  }]
 pelvis [{shapeType=Capsule, length=18.9027, radius=9.2081, center=<-4.8186 0 0>}]
 thigh_l [{shapeType=Capsule, length=18.1708, radius=9.7427, center=<-18.5736 0 0>, rotation=<rot -90 0 0>}]
 thigh_r [{shapeType=Capsule, length=17.9084, radius=10.0052, center=<18.5736 0 0>, rotation=<rot -90 180 0>}]
}
constraints [
 {
  bone1Name=dyn_shirt_shered_04_l
  bone2Name=base_shirt_shered_04_l
  swing1LimitDegrees=10
  swing2LimitDegrees=5
  twistLimitDegrees=45
  twistMotion=Locked
 }
 {
  bone1Name=base_shirt_shered_01_l
  bone2Name=pelvis
  swing1LimitDegrees=45
  swing1Motion=Free
  swing2LimitDegrees=45
  swing2Motion=Free
  twistLimitDegrees=45
  twistMotion=Free
 }
 {
  bone1Name=dyn_shirt_shered_01_l
  bone2Name=base_shirt_shered_01_l
  swing1LimitDegrees=10
  swing2LimitDegrees=5
  twistLimitDegrees=45
  twistMotion=Locked
 }
 {
  bone1Name=base_shirt_shered_01_r
  bone2Name=pelvis
  swing1LimitDegrees=45
  swing1Motion=Free
  swing2LimitDegrees=45
  swing2Motion=Free
  twistLimitDegrees=45
  twistMotion=Free
 }
 {
  bone1Name=dyn_shirt_shered_01_r
  bone2Name=base_shirt_shered_01_r
  swing1LimitDegrees=10
  swing2LimitDegrees=5
  twistLimitDegrees=45
  twistMotion=Locked
 }
 {
  bone1Name=base_shirt_shered_02_l
  bone2Name=pelvis
  swing1LimitDegrees=45
  swing1Motion=Free
  swing2LimitDegrees=45
  swing2Motion=Free
  twistLimitDegrees=45
  twistMotion=Free
 }
 {
  bone1Name=dyn_shirt_shered_02_l
  bone2Name=base_shirt_shered_02_l
  swing1LimitDegrees=10
  swing2LimitDegrees=5
  twistLimitDegrees=45
  twistMotion=Locked
 }
 {
  bone1Name=base_shirt_shered_02_r
  bone2Name=pelvis
  swing1LimitDegrees=45
  swing1Motion=Free
  swing2LimitDegrees=45
  swing2Motion=Free
  twistLimitDegrees=45
  twistMotion=Free
 }
 {
  bone1Name=dyn_shirt_shered_02_r
  bone2Name=base_shirt_shered_02_r
  swing1LimitDegrees=10
  swing2LimitDegrees=5
  twistLimitDegrees=45
  twistMotion=Locked
 }
 {
  bone1Name=base_shirt_shered_03_l
  bone2Name=pelvis
  swing1LimitDegrees=45
  swing1Motion=Free
  swing2LimitDegrees=45
  swing2Motion=Free
  twistLimitDegrees=45
  twistMotion=Free
 }
 {
  bone1Name=dyn_shirt_shered_03_l
  bone2Name=base_shirt_shered_03_l
  swing1LimitDegrees=10
  swing2LimitDegrees=5
  twistLimitDegrees=45
  twistMotion=Locked
 }
 {
  bone1Name=base_shirt_shered_03_r
  bone2Name=pelvis
  swing1LimitDegrees=45
  swing1Motion=Free
  swing2LimitDegrees=45
  swing2Motion=Free
  twistLimitDegrees=45
  twistMotion=Free
 }
 {
  bone1Name=dyn_shirt_shered_03_r
  bone2Name=base_shirt_shered_03_r
  swing1LimitDegrees=10
  swing2LimitDegrees=5
  twistLimitDegrees=45
  twistMotion=Locked
 }
 {
  bone1Name=base_shirt_shered_04_l
  bone2Name=pelvis
  swing1LimitDegrees=45
  swing1Motion=Free
  swing2LimitDegrees=45
  swing2Motion=Free
  twistLimitDegrees=45
  twistMotion=Free
 }
 {
  bone1Name=thigh_l
  bone2Name=pelvis
  swing1LimitDegrees=45
  swing1Motion=Free
  swing2LimitDegrees=45
  swing2Motion=Free
  twistLimitDegrees=45
  twistMotion=Free
 }
 {
  bone1Name=thigh_r
  bone2Name=pelvis
  swing1LimitDegrees=45
  swing1Motion=Free
  swing2LimitDegrees=45
  swing2Motion=Free
  twistLimitDegrees=45
  twistMotion=Free
 }
 {
  bone1Name=base_shirt_shered_01_c
  bone2Name=pelvis
  swing1LimitDegrees=45
  swing1Motion=Free
  swing2LimitDegrees=45
  swing2Motion=Free
  twistLimitDegrees=45
  twistMotion=Free
 }
 {bone1Name=dyn_shirt_shered_01_c, bone2Name=base_shirt_shered_01_c, swing1LimitDegrees=20, swing2LimitDegrees=20, twistLimitDegrees=20}
 {
  bone1Name=base_shirt_shered_04_r
  bone2Name=pelvis
  swing1LimitDegrees=45
  swing1Motion=Free
  swing2LimitDegrees=45
  swing2Motion=Free
  twistLimitDegrees=45
  twistMotion=Free
 }
 {bone1Name=dyn_shirt_shered_04_r, bone2Name=base_shirt_shered_04_r, swing1LimitDegrees=10, swing2LimitDegrees=0, twistLimitDegrees=0}
]
