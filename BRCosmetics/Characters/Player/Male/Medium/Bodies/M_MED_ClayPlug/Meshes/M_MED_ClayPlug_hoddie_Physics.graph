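# Physics asset for the hoodie of the ClayPlug medium male player body cosmetic, setting up secondary motion on its dyn bones.
# It holds two base-to-dyn bone pairs, hoodie and B_hoodie, each bone carrying one capsule body.
# Each dyn body is constrained to its base with narrow swing limits and zero twist limit; B_hoodie also zeroes one swing limit and has the largest capsule.
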PhysicsAsset /BRCosmetics/Characters/Player/Male/Medium/Bodies/M_MED_ClayPlug/Meshes/M_MED_ClayPlug_hoddie_Physics
bodies {
 base_B_hoodie [{shapeType=Capsule, length=0.505, radius=0.505, shapeName=base_B_hoodie_capsule, rotation=<rot 0 -90 0>}]
 base_hoodie [{shapeType=Capsule, length=0.505, radius=0.505, shapeName=base_hoodie_capsule, rotation=<rot 0 -90 0>}]
 dyn_B_hoodie [{shapeType=Capsule, length=6.9069, radius=3.4014, shapeName=dyn_B_hoodie_capsule, center=<1.2 0 0>, rotation=<rot -1.5169 -87.2552 -89.9999>}]
 dyn_hoodie [{shapeType=Capsule, length=2, radius=1, shapeName=dyn_hoodie_capsule, center=<2 0 0>, rotation=<rot 90 0 0>}]
}
constraints [
 {bone1Name=dyn_hoodie, bone2Name=base_hoodie, swing1LimitDegrees=8, swing2LimitDegrees=25, twistLimitDegrees=0}
 {bone1Name=dyn_B_hoodie, bone2Name=base_B_hoodie, swing1LimitDegrees=0, swing2LimitDegrees=10, twistLimitDegrees=0}
]
